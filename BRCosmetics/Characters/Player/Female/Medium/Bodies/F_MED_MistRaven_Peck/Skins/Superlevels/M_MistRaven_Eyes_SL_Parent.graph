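# Parent material for the eyes on the MistRaven female character's Superlevel skins, adding a parameterized hue shift to the eye glow.
# It uses material attributes, feeding the shared MF_eyes_2023 function through MF_TechArtOverrides.
# The emissive color is extracted, hue-shifted by the MistRavenCowl_EyesHueShift scalar in the superlevel group, and written back, leaving all other attributes intact.
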Material /BRCosmetics/Characters/Player/Female/Medium/Bodies/F_MED_MistRaven_Peck/Skins/Superlevels/M_MistRaven_Eyes_SL_Parent
graph {
 properties {
  bUseMaterialAttributes=true
  bUsedWithMeshDeformer=true
  bUsedWithSkeletalMesh=true
  pixelDepthOffsetMode=PDOM_Legacy
  shadingModel=MSM_FromMaterialExpression
  usedShadingModels=Default Lit
 }
 nodes [
  n1 = GetMaterialAttributes(None=MaterialFunctionCall(In=MaterialFunctionCall() {materialFunction=/Game/Global/Materials/Character_Parent_2023/Eye_Parent_2023/MF_eyes_2023}.Result) {materialFunction=&MF_TechArtOverrides}.Result) {attributeGetTypes [#1]}
  SetMaterialAttributes(
   MaterialAttributes=@n1.MaterialAttributes
   Emissive Color=MaterialFunctionCall(Hue Shift Percentage=ScalarParameter() {group=superlevel, parameterName=MistRavenCowl_EyesHueShift}, Texture=@n1.Emissive Color) {materialFunction=&HueShift}.Result
  ) {attributeSetTypes [#1]}
 ]
}
